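# Niagara system for the GenieSmoke vehicle drift trail cosmetic, attached to wheel bones on the car's skeletal mesh.
# Two ribbons form the main trail and meshes wrap the wheels, while three sprite emitters add curl-noise specks, orbiting particles and SubUV puffs.
# Each emitter masks particles per tire through FN_DisableTireFX.
# Painted-variant modules recolor the effect, with user flags switching default versus painted looks and adjusting the ribbon for frontend preview.
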NiagaraSystem /VehicleCosmetics/DriftTrail/DriftTrail_GenieSmoke/FX/NS_DriftTrail_GenieSmoke_MP
dependencies {
 usedDataInterfaces [
  &NiagaraDataInterfaceColorCurve
  &NiagaraDataInterfaceCurve
  &NiagaraDataInterfaceArrayColor
  &NiagaraDataInterfaceSpriteRendererInfo
  /Script/Niagara.NiagaraDataInterfaceArrayFloat4
  &NiagaraDataInterfaceSkeletalMesh
 ]
 usedDynamicInputs [
  &Multiply_Vector2D
  &Vector2DFromFloat
  /Niagara/DynamicInputs/Arrays/SelectVector4_FromArray
  &MakeIntFromFloat
 ]
 usedModules [
  &self:SystemSpawnScript.SystemScriptSource.SystemScriptGraph.NiagaraNodeAssignment_0.SetVariables_#1
  &FN_UnpackVehicleFlags
  /VehicleCosmetics/FX/Niagara/Modules/VC_SetVehicleSpeedAttribs
  &SystemState
 ]
}
emitters {
 NE_VC_DriftTrail_Sprite_Parent {
  emitterUpdateScript [
   &EmitterState()
   &SpawnRate(SpawnRate=70, Use Spawn Probability=false)
  ]
  particleSpawnScript [
   &VC_PaintedVariantsOverride(
    Black=<rgba 0.01 0.01 0.01 25>
    BurntSienna=<rgba 0.0642 0.0107 0 2>
    Cobalt=<rgba 0 0.925 50 1>
    Crimson=<rgba 20 0 0 2>
    DefaultColor=&SelectLinearColorFromArray(
     Array Sampling Mode=::NewEnumerator0
     Direct Array Index=?
     Color Selection Array=DataInterface() {propertyValues {ColorData [<rgba 20 3 14.8121 1>, <rgba 10 15.2506 20 1>]}}
     Random Seed=0
    )
    ForestGreen=<rgba 0 10 0 1>
    Grey=<rgba 5 5 5 1>
    Lime=<rgba 5 10 0 1.5>
    Orange=<rgba 10 0.3333 0 1>
    Pink=<rgba 5 0 3.75 2>
    Purple=<rgba 0.5 0 1 2>
    Saffron=<rgba 10 6.6667 0 1>
    SkyBlue=<rgba 0 1.3308 2 1>
    TitaniumWhite=<rgba 2 2 2 1>
   )
   &InitializeParticle(
    Color Mode=::NewEnumerator0
    Lifetime Max=0.5
    Lifetime Min=0.3
    Lifetime Mode=::NewEnumerator1
    Sprite Rotation Mode=::NewEnumerator2
    Sprite Size Max=<2 2.5>
    Sprite Size Min=<1.5 2>
    Sprite Size Mode=::NewEnumerator2
    UsePositionOffset=false
    Write Material Random=true
   )
   &PartitionParticles()
   &SkeletalMeshLocation(
    Bone / Socket Index=&Add_Integer(A=$Particles.PartitionParticles.Partition, B=4)
    Bone / Socket Sampling Mode=::NewEnumerator5
    Interpolate Mesh Sampling=false
    MaskSpawnGroup=false
    Mesh Sampling Type=::NewEnumerator6
    Sampled Position Offset=&AddVectorToPosition(Position=<-60 0 -25>, Vector=$User.offset)
    Skeletal Mesh=$System.SkeletalMesh
   )
   &ShapeLocation(MaskSpawnGroup=false, Override Seed=false, Sphere Radius=20, UseOwnerScale=true)
   &AddVelocity(Velocity=&RandomRangeVector(Minimum=<-350 0 15>, Maximum=<-300 0 -15>))
  ]
  particleUpdateScript [
   &FN_DisableTireFX(
    Mask Index Attribute:NiagaraInt32=$Particles.PartitionParticles.Partition
    Mask Index FL=2
    Mask Index FR=3
    Mask Index RL=0
    Mask Index RR=1
   )
   &ParticleState()
   &SpriteRotationRate(Rotation Rate=20)
   &InheritSourceMovement(
    Applied Position Scale=&Multiply_VectorByFloat(
     Vector=<1 1 1>
     Float=&FloatFromCurve(FloatCurve=DataInterface() {propertyValues {Curve {keys [{interpMode=RCIM_Cubic, tangentMode=RCTM_Break, time=0.3, value=1}, {interpMode=RCIM_Cubic, tangentMode=RCTM_Break, time=1, value=0}]}}})
    )
    ApplyForce=false
    ApplyPosition=true
    ApplyPositionFalloff=false
   )
   &CurlNoiseForce(Bias Noise=false, Noise Frequency=10, Noise Strength=100, Pan Noise=false)
   &SolveForcesAndVelocity(Clamp Velocity=false, Limit Acceleration=false)
   &ScaleColor(
    Scale Alpha=&FloatFromCurve(FloatCurve=DataInterface() {propertyValues {Curve {
        defaultValue=1
        keys [
         <at 0 0 RCIM_Cubic>
         {arriveTangent=0.1714, interpMode=RCIM_Cubic, leaveTangent=0.1714, tangentMode=RCTM_User, time=0.1, value=1}
         {arriveTangent=-0.0405, interpMode=RCIM_Cubic, leaveTangent=-0.0405, tangentMode=RCTM_User, time=0.5, value=1}
         {arriveTangent=-4.0173, interpMode=RCIM_Cubic, leaveTangent=-4.0173, tangentMode=RCTM_User, time=1, value=0}
        ]
       }}})
    Scale RGB=&VectorFromFloat(Value=&FloatFromCurve(FloatCurve=DataInterface() {propertyValues {Curve {keys [<at 0.5 1 RCIM_Cubic>,<at 1 0.5 RCIM_Cubic>]}}}, Scale Curve=50))
    ScaleA=true
    ScaleRGB=true
   )
  ]
  rendererClasses [&NiagaraSpriteRendererProperties]
  renderers [{rendererClass=&NiagaraSpriteRendererProperties}]
 }
 NE_VC_DriftTrail_Sprite_Parent001 {
  emitterUpdateScript [
   &EmitterState(Life Cycle Mode=::NewEnumerator1, Loop Duration=2, UseLoopDelay=false)
   &SpawnRate(SpawnRate=50, Use Spawn Probability=false)
   &SpawnRate(SpawnGroup=1, SpawnRate=20, Use Spawn Probability=false) {module=SpawnRate001}
  ]
  particleSpawnScript [
   &VC_PaintedVariantsOverride(
    Black=<rgba 0.005 0.005 0.005 50>
    BurntSienna=<rgba 0.02 0.0027 0 100>
    Cobalt=<rgba 0 0.925 50 1>
    Crimson=<rgba 10 0 0.8333 1>
    DefaultColor=<rgba 1 0 0.991 1>
    ForestGreen=<rgba 0 2.5 0.0833 2>
    Grey=<rgba 0.5 0.5 0.5 1>
    Lime=<rgba 1 2 0 1>
    Orange=<rgba 5 0.1667 0 1>
    Pink=<rgba 2 0.04 1.2873 1>
    Purple=<rgba 2.5 0 10 1>
    Saffron=<rgba 3 2 0 1>
    SkyBlue=<rgba 0 2.0833 5 1>
    TitaniumWhite=<rgba 2 2 2 1>
   )
   &InitializeParticle(
    Color Mode=::NewEnumerator0
    Lifetime=/VehicleCosmetics/FX/Niagara/Modules/MergeEmitters/VC_FloatBySpawnGroup(
     SpawnGroup or ExecIndex:NiagaraInt32=0
     CustomInt:NiagaraInt32=0
     MaximumSpawnGroups:NiagaraInt32=2
     Float SG 0:NiagaraFloat=&RandomRangeFloat(Minimum=0.1, Maximum=0.05)
     Float SG 1:NiagaraFloat=&RandomRangeFloat(Minimum=0.1, Maximum=0.32)
     Float SG 2:NiagaraFloat=&RandomRangeFloat(Minimum=1.5, Maximum=2)
     Float SG 3:NiagaraFloat=0
     Float SG 4:NiagaraFloat=0
     Float SG 5:NiagaraFloat=0
    )
    Sprite Rotation Mode=::NewEnumerator2
    Sprite Size=&MakeVector2D(
     X=&VC_FloatBySpawnGroup(
      SpawnGroup or ExecIndex:NiagaraInt32=0
      CustomInt:NiagaraInt32=0
      MaximumSpawnGroups:NiagaraInt32=2
      Float SG 0:NiagaraFloat=&RandomRangeFloat(Minimum=3, Maximum=6)
      Float SG 1:NiagaraFloat=&RandomRangeFloat(Minimum=10, Maximum=14)
      Float SG 2:NiagaraFloat=60
      Float SG 3:NiagaraFloat=0
      Float SG 4:NiagaraFloat=0
      Float SG 5:NiagaraFloat=0
     )
     Y=&VC_FloatBySpawnGroup(
      SpawnGroup or ExecIndex:NiagaraInt32=0
      CustomInt:NiagaraInt32=0
      MaximumSpawnGroups:NiagaraInt32=2
      Float SG 0:NiagaraFloat=&RandomRangeFloat(Minimum=6, Maximum=12)
      Float SG 1:NiagaraFloat=&RandomRangeFloat(Minimum=10, Maximum=14)
      Float SG 2:NiagaraFloat=60
      Float SG 3:NiagaraFloat=0
      Float SG 4:NiagaraFloat=0
      Float SG 5:NiagaraFloat=0
     )
    )
    Sprite Size Mode=::NewEnumerator4
    UsePositionOffset=false
    Write Material Random=true
   )
   &PartitionParticles(Particles Per Partition=2, Partition Mode:ENiagara_PartitionMode=::NewEnumerator1)
   &SkeletalMeshLocation(
    Bone / Socket Index=&Add_Integer(A=$Particles.PartitionParticles.Partition, B=4)
    Bone / Socket Sampling Mode=::NewEnumerator5
    Interpolate Mesh Sampling=false
    MaskSpawnGroup=false
    Mesh Sampling Type=::NewEnumerator6
    Sampled Position Offset=&AddVectorToPosition(Position=<-65 0 -25>, Vector=$User.offset)
    Skeletal Mesh=$System.SkeletalMesh
   )
   &AddVelocity(Velocity=<1000 0 50>, Velocity Speed Scale=-1) {enabled=false}
   /SetVariables(Particles.RenderVisTag:NiagaraInt32=&SpawnGroup()) {isSetParameters=true}
   &ShapeLocation(MaskSpawnGroup=false, Non Uniform Scale=<1 0 0>, Override Seed=false, Sphere Radius=30, Sphere Surface Distribution=0.5, UseOwnerScale=true)
  ]
  particleUpdateScript [
   &FN_DisableTireFX(
    Mask Index Attribute:NiagaraInt32=$Particles.PartitionParticles.Partition
    Mask Index FL=2
    Mask Index FR=3
    Mask Index RL=0
    Mask Index RR=1
   )
   &ParticleState()
   &ScaleSpriteSize(Uniform Curve Sprite Scale=DataInterface() {propertyValues {Curve {
       defaultValue=0
       keys [
        {arriveTangent=4.1785, interpMode=RCIM_Cubic, leaveTangent=4.1785, tangentMode=RCTM_User, time=0, value=0}
        {arriveTangent=0.0224, interpMode=RCIM_Cubic, leaveTangent=0.0224, tangentMode=RCTM_User, time=0.5, value=1}
       ]
      }}})
   &RotateAroundPoint(
    Radius:NiagaraFloat=&RandomRangeFloat(Minimum=14, Maximum=20)
    Rotation Center:NiagaraPosition=$Particles.Initial.Position
    Rotation Rate:NiagaraFloat=&VC_FloatBySpawnGroup(
     SpawnGroup or ExecIndex:NiagaraInt32=0
     CustomInt:NiagaraInt32=0
     MaximumSpawnGroups:NiagaraInt32=2
     Float SG 0:NiagaraFloat=0.8
     Float SG 1:NiagaraFloat=0.1
     Float SG 2:NiagaraFloat=0
     Float SG 3:NiagaraFloat=0
     Float SG 4:NiagaraFloat=0
     Float SG 5:NiagaraFloat=0
    )
    X Axis:Vector3f=/VehicleCosmetics/FX/Niagara/Modules/MergeEmitters/VC_VectorBySpawnGroup(
     SpawnGroup or ExecIndex:NiagaraInt32=0
     CustomInt:NiagaraInt32=0
     MaximumSpawnGroups:NiagaraInt32=2
     Vector Group 0:Vector3f=&RandomRangeVector(Minimum=<-1 -1 -1>, Maximum=<1 1 1>)
     Vector Group 1:Vector3f=&RandomRangeVector(Minimum=<-1 -1 -1>, Maximum=<1 1 1>)
     Vector Group 2:Vector3f=<0 0 0>
     Vector Group 3:Vector3f=<0 0 0>
     Vector Group 4:Vector3f=<0 0 0>
     Vector Group 5:Vector3f=<0 0 0>
    )
    Y Axis:Vector3f=&RandomRangeVector(Minimum=<-1 -1 -1>, Maximum=<1 1 1>)
   )
   &SpriteRotationRate(Rotation Rate=&VC_FloatBySpawnGroup(
     SpawnGroup or ExecIndex:NiagaraInt32=0
     CustomInt:NiagaraInt32=0
     MaximumSpawnGroups:NiagaraInt32=2
     Float SG 0:NiagaraFloat=0
     Float SG 1:NiagaraFloat=30
     Float SG 2:NiagaraFloat=200
     Float SG 3:NiagaraFloat=0
     Float SG 4:NiagaraFloat=0
     Float SG 5:NiagaraFloat=0
    ))
   &SolveForcesAndVelocity(Clamp Velocity=false, Limit Acceleration=false)
   &ScaleColor(ScaleA=true, ScaleRGB=true)
   /SetVariables(
    Particles.R0_ScaledColor:LinearColor=&VC_DynamicScaleColor(
     Color to Scale:LinearColor=$Particles.Initial.Color
     RGB Scale:Vector3f=&VectorFromFloat(Value=&MakeCustomFloatFromBool(Bool=$User.bIsPainted, True Float=50, False Float=500))
     Alpha Scale:NiagaraFloat=&FloatFromCurve(
      FloatCurve=DataInterface() {propertyValues {Curve {
         defaultValue=1
         keys [
          {interpMode=RCIM_Cubic, tangentMode=RCTM_User, time=0, value=0}
          {interpMode=RCIM_Cubic, tangentMode=RCTM_User, time=0.25, value=1}
          {arriveTangent=-0.1057, interpMode=RCIM_Cubic, leaveTangent=-0.1057, tangentMode=RCTM_User, time=0.75, value=1}
          {tangentMode=RCTM_User, time=1, value=0}
         ]
        }}}
      Scale Curve=2
     )
    )
    Particles.R1_ScaledColor:LinearColor=&VC_DynamicScaleColor(
     Color to Scale:LinearColor=$Particles.Initial.Color
     RGB Scale:Vector3f=&VectorFromFloat(Value=500)
     Alpha Scale:NiagaraFloat=&FloatFromCurve(FloatCurve=DataInterface() {propertyValues {Curve {
         defaultValue=1
         keys [
          {interpMode=RCIM_Cubic, tangentMode=RCTM_User, time=0, value=0}
          {interpMode=RCIM_Cubic, tangentMode=RCTM_User, time=0.5, value=1}
          {tangentMode=RCTM_User, time=1, value=0}
         ]
        }}})
    )
    Particles.R1_SpriteSize:Vector2f=&Multiply_Vector2D_ByFloat(
     Float=&FloatFromCurve(
      FloatCurve=DataInterface() {propertyValues {Curve {defaultValue=1, keys [{tangentMode=RCTM_User, time=0, value=0}, {tangentMode=RCTM_User, time=1, value=1}]}}}
      Scale Curve=&RandomRangeFloat(Minimum=0.75, Maximum=1)
     )
     Vector2D=$Particles.Initial.SpriteSize
    )
   ) {isSetParameters=true}
  ]
  rendererClasses [&NiagaraSpriteRendererProperties]
  renderers [
   {rendererClass=&NiagaraSpriteRendererProperties}
   {rendererClass=&NiagaraSpriteRendererProperties, rendererIndex=1}
  ]
 }
 NE_VC_DriftTrail_Sprite_Parent002 {
  emitterUpdateScript [
   &EmitterState(Life Cycle Mode=::NewEnumerator1, Loop Duration=2, UseLoopDelay=false)
   &SpawnRate(
    SpawnRate=/VehicleCosmetics/FX/Niagara/Modules/VC_Mask_Floats(
     FrontEnd Value:NiagaraFloat=5
     blsFrontEndPreview:NiagaraBool=false
     SpeedMin Value:NiagaraFloat=5
     SpeedMax Value:NiagaraFloat=5
     Speed Alpha:NiagaraFloat=$System.Speed_Alpha
    )
    Use Spawn Probability=false
   )
  ]
  particleSpawnScript [
   &VC_PaintedVariantsOverride(
    Black=<rgba 0.07 0.07 0.07 5>
    BurntSienna=<rgba 0.5 0.1 0 2>
    Cobalt=<rgba 0 0.0625 1.25 1.5>
    Crimson=<rgba 1.25 0.0625 0.0947 1>
    DefaultColor=<rgba 1 0.099 0.9919 1>
    ForestGreen=<rgba 0.05 1 0.05 1>
    Grey=<rgba 0.55 0.55 0.55 1>
    Lime=<rgba 0.15 0.3 0 2.5>
    Orange=<rgba 1 0.1667 0 1.2>
    Pink=<rgba 1 0.1198 0.6799 1>
    Purple=<rgba 0.16 0 0.8 2>
    R1 DefaultColor=<rgba 1 0 0.991 1>
    Saffron=<rgba 1.1 0.8433 0 1>
    SkyBlue=<rgba 0 0.6654 1 1>
    Use R1 Painted Overrides=true
   )
   &InitializeParticle(
    Color Mode=::NewEnumerator0
    Lifetime Max=1
    Lifetime Mode=::NewEnumerator1
    Sprite Rotation Mode=::NewEnumerator2
    Sprite Size Mode=::NewEnumerator1
    Uniform Sprite Size Max=40
    Uniform Sprite Size Min=25
    UsePositionOffset=false
    Write Material Random=true
   )
   &PartitionParticles()
   &SkeletalMeshLocation(
    Bone / Socket Index=&Add_Integer(A=$Particles.PartitionParticles.Partition, B=4)
    Bone / Socket Sampling Mode=::NewEnumerator5
    Interpolate Mesh Sampling=false
    MaskSpawnGroup=false
    Mesh Sampling Type=::NewEnumerator6
    Sampled Position Offset=&AddVectorToPosition(Position=<-45 0 -25>, Vector=$User.offset)
    Skeletal Mesh=$System.SkeletalMesh
   )
   &ShapeLocation(MaskSpawnGroup=false, Override Seed=false, Sphere Radius=15, UseOwnerScale=true)
   &AddVelocity(Velocity=&VC_Mask_Vectors(
     FrontEnd Value:Vector3f=&RandomRangeVector(Minimum=<-15 -60 -50>, Maximum=<-170 60 50>)
     blsFrontEndPreview:NiagaraBool=false
     SpeedMin Value:Vector3f=&RandomRangeVector(Minimum=<-15 -50 -50>, Maximum=<-170 50 50>)
     SpeedMax Value:Vector3f=&RandomRangeVector(Minimum=<-15 -50 -50>, Maximum=<-170 50 50>)
     Speed Alpha:NiagaraFloat=$System.Speed_Alpha
    ))
  ]
  particleUpdateScript [
   &FN_DisableTireFX(
    Mask Index Attribute:NiagaraInt32=$Particles.PartitionParticles.Partition
    Mask Index FL=2
    Mask Index FR=3
    Mask Index RL=0
    Mask Index RR=1
   )
   &ParticleState()
   &SpriteRotationRate(Rotation Rate=200)
   &ScaleSpriteSize(Uniform Curve Sprite Scale=DataInterface() {propertyValues {Curve {
       defaultValue=0
       keys [
        {arriveTangent=4.1785, interpMode=RCIM_Cubic, leaveTangent=4.1785, tangentMode=RCTM_User, time=0, value=0}
        {arriveTangent=0.0224, interpMode=RCIM_Cubic, leaveTangent=0.0224, tangentMode=RCTM_User, time=0.5, value=1}
       ]
      }}})
   &Drag(
    Drag=&FloatFromCurve(FloatCurve=DataInterface() {propertyValues {Curve {defaultValue=1, keys [{tangentMode=RCTM_User, time=0, value=1}, {tangentMode=RCTM_User, time=1, value=2}]}}})
    Use Linear Drag=true
    Use Rotational Drag=true
   )
   &SolveForcesAndVelocity(Clamp Velocity=false, Limit Acceleration=false)
   &ScaleColor(
    Scale Alpha=&FloatFromCurve(FloatCurve=DataInterface() {propertyValues {Curve {keys [<at 0 1>,<at 1 0>]}}}, Scale Curve=0.5)
    Scale RGB=&VectorFromFloat(Value=25)
    ScaleA=true
    ScaleRGB=true
   )
   &SubUVAnimation(
    End Frame Range Override=7
    "Sprite Renderer "=DataInterface() {propertyValues {SpriteRenderer=&self:NE_VC_DriftTrail_Sprite_Parent002_0.NiagaraSpriteRendererProperties_4}}
    UseEndFrame=true
    UseStartFrame=true
   )
   &DynamicMaterialParameters(
    Index 0 Param 1=$StackContext.SubImageIndex
    Module.Index0Float0Write=true
    Module.Index0Float1Write=true
    Module.Index0Float2Write=true
    Module.Index0Float3Write=true
   )
  ]
  rendererClasses [&NiagaraSpriteRendererProperties]
  renderers [{rendererClass=&NiagaraSpriteRendererProperties}]
 }
 NE_VC_MainRibbon {
  emitterUpdateScript [
   &EmitterState()
   &VC_RateQualityHandler(Epic Spawn Rate=30, High Spawn Rate=30, Low Spawn Rate=30, Use Custom Rate?=true)
   &VC_SpawnRibbon_v1(Number of Ribbons=2, SpawnRate=$Emitter.Spawn Rate)
  ]
  particleSpawnScript [
   &VC_PaintedVariantsOverride(Use R1 Painted Overrides=true) {enabled=false}
   &VC_InitializeRibbon_v1(
    Custom Offset=true
    Enable Skeletal Position=true
    Lifetime=0.75
    Local Vector Offset=&AddVectorToPosition(Position=<-35 0 -25>, Vector=$User.offset)
    Offset Amount=1
    Ribbon Width=40
   )
   &AddVelocity(
    Velocity=&MakeCustomVectorFromBool(Bool=$User.bIsFrontendPreview, True Vector=<-1000 0 0>, False Vector=<-500 0 0>)
    Velocity Speed Scale=1.15
   ) {module=AddVelocity001}
   /SetVariables(Particles.WidthComple=70) {isSetParameters=true}
  ]
  particleUpdateScript [
   &FN_DisableTireFX(Mask Index Attribute:NiagaraInt32=$Particles.SpawnGroup, Mask Index FL=2, Mask Index FR=3, Mask Index RL=0, Mask Index RR=1)
   &ParticleState()
   &VC_PaintedVariantsOverride(
    Black=<rgba 0.2 0.2 0.2 2>
    BurntSienna=<rgba 0.5 0.1347 0.0616 1>
    Cobalt=<rgba 0.085 0.2196 1.7 1>
    Crimson=<rgba 1.5 0.09 0.2075 1>
    ForestGreen=<rgba 0.06 0.3 0.0721 1>
    Grey=<rgba 0.25 0.25 0.25 1>
    Lime=<rgba 0.15 0.3 0 1>
    Orange=<rgba 0.6724 0.1479 0.0672 1>
    Pink=<rgba 0.7 0.175 0.5687 1>
    Purple=<rgba 0.4375 0.25 2.5 1>
    R1 Black:LinearColor=&ColorFromCurve(ColorCurve=DataInterface() {propertyValues {
       AlphaCurve {keys [<at 0 0>, <at 0.0324 0.6>, {arriveTangent=25.218, interpMode=RCIM_Cubic, leaveTangent=25.218, time=0.5357, value=1.5}, <at 1 25>]}
       BlueCurve {keys [
         <at 0.0691 0.45>
         <at 0.3679 0.27>
         {arriveTangent=-0.3072, interpMode=RCIM_Cubic, leaveTangent=-0.3072, time=0.5435, value=0.12}
         <at 0.9865 0.08>
        ]}
       GreenCurve {keys [
         <at 0.0691 0.45>
         <at 0.3679 0.27>
         {arriveTangent=-0.3072, interpMode=RCIM_Cubic, leaveTangent=-0.3072, time=0.5435, value=0.12}
         <at 0.9865 0.08>
        ]}
       RedCurve {keys [
         <at 0.0691 0.45>
         <at 0.3679 0.27>
         {arriveTangent=-0.3072, interpMode=RCIM_Cubic, leaveTangent=-0.3072, time=0.5435, value=0.12}
         <at 0.9865 0.08>
        ]}
      }})
    R1 BurntSienna:LinearColor=&ColorFromCurve(ColorCurve=DataInterface() {propertyValues {
       AlphaCurve {keys [<at 0 2 RCIM_Cubic>,<at 0.1871 0.85>,<at 0.4964 1>,<at 1 25 RCIM_Cubic>]}
       BlueCurve {keys [<at 0 0.16 RCIM_Cubic>,<at 0.2238 0.252>,<at 0.5435 0.0772>,<at 1 0>]}
       GreenCurve {keys [<at 0 0.208 RCIM_Cubic>,<at 0.2238 0.2984>,<at 0.5435 0.099>,<at 1 0.0042>]}
       RedCurve {keys [<at 0 0.4 RCIM_Cubic>,<at 0.2238 0.6>,<at 0.5435 0.2411>,<at 1 0.05>]}
      }})
    R1 Cobalt:LinearColor=&ColorFromCurve(ColorCurve=DataInterface() {propertyValues {
       AlphaCurve {keys [<at 0 2 RCIM_Cubic>,<at 0.2089 0.6>,<at 1 0.8 RCIM_Cubic>]}
       BlueCurve {keys [<at 0 0.8469>,<at 0.2304 0.7>,<at 0.5703 0.5>,<at 1 0.7>]}
       GreenCurve {keys [<at 0 0.2823>,<at 0.2304 0.2333>,<at 0.5703 0.1562>,<at 1 0.2567>]}
       RedCurve {keys [<at 0 0.1694>,<at 0.2304 0.14>,<at 0.5703 0.125>,<at 1 0.035>]}
      }})
    R1 Crimson:LinearColor=&ColorFromCurve(ColorCurve=DataInterface() {propertyValues {
       AlphaCurve {keys [<at 0 0.5>,<at 0.4606 0.6>,<at 0.6542 1.5>,<at 1 10 RCIM_Cubic>]}
       BlueCurve {keys [<at 0 0.2229 RCIM_Cubic>,<at 0.4821 0.1333>,<at 0.9209 0.0203>]}
       GreenCurve {keys [<at 0 0.1875 RCIM_Cubic>,<at 0.4821 0.1>,<at 0.9209 0.004>]}
       RedCurve {keys [<at 0 1.25 RCIM_Cubic>,<at 0.4821 0.5>,<at 0.9209 0.2>]}
      }})
    R1 DefaultColor=&ColorFromCurve(ColorCurve=DataInterface() {propertyValues {
       AlphaCurve {keys [<at 0 0.5>,<at 0.1714 1>,<at 0.3165 0.8>,<at 0.51 0.5>,<at 0.5681 4>,<at 0.9975 0>]}
       BlueCurve {keys [
         <at 0 0.5>
         {arriveTangent=1.8613, interpMode=RCIM_Cubic, leaveTangent=1.8613, time=0.1695, value=1}
         <at 0.5531 1.25>
         <at 0.8059 2 RCIM_Cubic>
        ]}
       GreenCurve {keys [
         <at 0 0.5>
         {arriveTangent=1.8613, interpMode=RCIM_Cubic, leaveTangent=1.8613, time=0.1695, value=1}
         <at 0.5531 1.25>
         <at 0.8059 2 RCIM_Cubic>
        ]}
       RedCurve {keys [
         <at 0 0.5>
         {arriveTangent=1.8613, interpMode=RCIM_Cubic, leaveTangent=1.8613, time=0.1695, value=1}
         <at 0.5531 1.25>
         <at 0.8059 2 RCIM_Cubic>
        ]}
      }})
    R1 ForestGreen:LinearColor=&ColorFromCurve(ColorCurve=DataInterface() {propertyValues {
       AlphaCurve {keys [<at 0 1 RCIM_Cubic>,<at 0.5767 1>,<at 1 10 RCIM_Cubic>]}
       BlueCurve {keys [<at 0.2218 0.2451>,<at 0.5703 0.084>,<at 1 0.0051>]}
       GreenCurve {keys [<at 0.2218 0.57>,<at 0.5703 0.2502>,<at 1 0.1>]}
       RedCurve {keys [<at 0.2218 0.228>,<at 0.5703 0.075>,<at 1 0>]}
      }})
    R1 Grey:LinearColor=&ColorFromCurve(ColorCurve=DataInterface() {propertyValues {
       AlphaCurve {keys [<at 0 1 RCIM_Cubic>,<at 0.4477 0.7>,<at 0.5832 1>,<at 0.7294 2>,<at 1 10 RCIM_Cubic>]}
       BlueCurve {keys [<at 0 0.55 RCIM_Cubic>,<at 0.5509 0.22>,<at 1 0.25>]}
       GreenCurve {keys [<at 0 0.55 RCIM_Cubic>,<at 0.5509 0.22>,<at 1 0.25>]}
       RedCurve {keys [<at 0 0.55 RCIM_Cubic>,<at 0.5509 0.22>,<at 1 0.25>]}
      }})
    R1 Lime:LinearColor=&ColorFromCurve(ColorCurve=DataInterface() {propertyValues {
       AlphaCurve {keys [<at 0 0.5 RCIM_Cubic>,<at 0.3466 0.7>,<at 0.7144 2>,<at 1 5 RCIM_Cubic>]}
       BlueCurve {keys [<at 0 0>,<at 0.452 0.0105>,<at 1 0>]}
       GreenCurve {keys [<at 0 0.85>,<at 0.452 0.35>,<at 1 0.25>]}
       RedCurve {keys [<at 0 0.6092>,<at 0.452 0.2368>,<at 1 0.1254>]}
      }})
    R1 Orange:LinearColor=&ColorFromCurve(ColorCurve=DataInterface() {propertyValues {
       AlphaCurve {keys [
         <at 0.0153 0>
         {arriveTangent=1.8596, interpMode=RCIM_Cubic, leaveTangent=1.8596, time=0.1078, value=0.5}
         <at 0.3917 0.7>
         <at 0.6757 2>
         <at 1 10 RCIM_Cubic>
        ]}
       BlueCurve {keys [<at 0 0>,<at 0.4428 0.135>,<at 0.8435 0.04>]}
       GreenCurve {keys [<at 0 0.4667>,<at 0.4428 0.1613>,<at 0.8435 0.0533>]}
       RedCurve {keys [<at 0 1.4>,<at 0.4428 0.45>,<at 0.8435 0.2>]}
      }})
    R1 Pink:LinearColor=&ColorFromCurve(ColorCurve=DataInterface() {propertyValues {
       AlphaCurve {keys [<at 0 0.2>,<at 0.1596 0.5>,<at 0.4633 0.6>,<at 1 25>]}
       BlueCurve {keys [
         <at 0.0024 0.8178>
         {arriveTangent=-1, interpMode=RCIM_Cubic, leaveTangent=-1, time=0.1325, value=0.7449}
         <at 0.4444 0.3757>
         <at 0.9216 0.079>
        ]}
       GreenCurve {keys [
         <at 0.0024 0.5>
         {arriveTangent=-0.6731, interpMode=RCIM_Cubic, leaveTangent=-0.6731, time=0.1325, value=0.3}
         <at 0.4444 0.2025>
         <at 0.9216 0.03>
        ]}
       RedCurve {keys [<at 0.0024 1>, {arriveTangent=-1.2444, interpMode=RCIM_Cubic, leaveTangent=-1.2444, time=0.1325, value=1}, <at 0.4444 0.45>, <at 0.9216 0.1>]}
      }})
    R1 Purple:LinearColor=&ColorFromCurve(ColorCurve=DataInterface() {propertyValues {
       AlphaCurve {keys [<at 0.0729 1 RCIM_Cubic>,<at 0.5284 1>,<at 1 10 RCIM_Cubic>]}
       BlueCurve {keys [
         <at 0.0078 1>
         {arriveTangent=-0.8896, interpMode=RCIM_Cubic, leaveTangent=-0.8896, time=0.203, value=0.8}
         <at 0.626 0.45>
         <at 0.7399 0.5>
         <at 0.9433 0.15>
        ]}
       GreenCurve {keys [
         <at 0.0078 0>
         {arriveTangent=0.0728, interpMode=RCIM_Cubic, leaveTangent=0.0728, time=0.203, value=0}
         <at 0.626 0.045>
         <at 0.7399 0>
         <at 0.9433 0>
        ]}
       RedCurve {keys [
         <at 0.0078 0.4167>
         {arriveTangent=-0.4811, interpMode=RCIM_Cubic, leaveTangent=-0.4811, time=0.203, value=0.2667}
         <at 0.626 0.1193>
         <at 0.7399 0.125>
         <at 0.9433 0.025>
        ]}
      }})
    R1 Saffron:LinearColor=&ColorFromCurve(ColorCurve=DataInterface() {propertyValues {
       AlphaCurve {keys [<at 0 0>, <at 0.062 1>, {arriveTangent=3.2162, interpMode=RCIM_Cubic, leaveTangent=3.2162, time=0.3955, value=0.6}, <at 0.9948 4>]}
       BlueCurve {keys [<at 0 0.07>,<at 0.4932 0.1>,<at 0.683 0.0496>,<at 0.965 0>]}
       GreenCurve {keys [<at 0 0.616>,<at 0.4932 0.23>,<at 0.683 0.1845>,<at 0.965 0.3333>]}
       RedCurve {keys [<at 0 0.7>,<at 0.4932 0.25>,<at 0.683 0.2>,<at 0.965 0.4>]}
      }})
    R1 SkyBlue:LinearColor=&ColorFromCurve(ColorCurve=DataInterface() {propertyValues {
       AlphaCurve {keys [
         <at 0 0>
         <at 0.0919 0.5>
         {arriveTangent=1.0025, interpMode=RCIM_Cubic, leaveTangent=1.0025, time=0.3359, value=0.5}
         <at 0.9975 1 RCIM_Cubic>
        ]}
       BlueCurve {keys [<at 0 1.3>,<at 0.4254 0.55>,<at 0.7399 0.16>,<at 1 0.7>]}
       GreenCurve {keys [<at 0 0.8233>,<at 0.4254 0.2567>,<at 0.7399 0.1067>,<at 1 0.4433>]}
       RedCurve {keys [<at 0 0>,<at 0.4254 0>,<at 0.7399 0>,<at 1 0>]}
      }})
    R1 TitaniumWhite:LinearColor=&ColorFromCurve(ColorCurve=DataInterface() {propertyValues {
       AlphaCurve {keys [<at 0.0051 0>, {arriveTangent=2.9386, interpMode=RCIM_Cubic, leaveTangent=2.9386, time=0.1461, value=0.2}, <at 0.5257 0.7>, <at 0.6857 2>]}
       BlueCurve {keys [<at 0 1.2 RCIM_Cubic>,<at 0.5257 0.35>,<at 0.9948 0.35>]}
       GreenCurve {keys [<at 0 1.2 RCIM_Cubic>,<at 0.5257 0.35>,<at 0.9948 0.35>]}
       RedCurve {keys [<at 0 1.2 RCIM_Cubic>,<at 0.5257 0.35>,<at 0.9948 0.35>]}
      }})
    Saffron=<rgba 0.4 0.3353 0.012 1>
    SkyBlue=<rgba 0 0.3333 1 1>
    TitaniumWhite=<rgba 0.47 0.47 0.47 1>
    Use R1 Painted Overrides=true
   ) {module=VC_PaintedVariantsOverride002}
   &DynamicMaterialParameters(
    Index 0 Param 1=&MakeCustomFloatFromBool(Bool=$User.bIsPainted, True Float=0, False Float=0)
    Index 0 Param 2=&MakeCustomFloatFromBool(Bool=$User.bIsPainted, True Float=0, False Float=2.8)
    Index 0 Param 3=&SelectFloatFromArray(Direct Array Index=&SpawnGroup(), Float Selection Array=DataInterface() {propertyValues {FloatData [0.3,0]}})
    Index 0 Param 4=&FloatFromCurve(FloatCurve=DataInterface() {propertyValues {Curve {defaultValue=1, keys [<at 0 0.15 RCIM_Cubic>, {arriveTangentWeight=0.6604, interpMode=RCIM_Cubic, tangentMode=RCTM_User, time=1, value=0.2}]}}})
    Index 1 Linear Color:LinearColor=&MakeCustomLinearColorFromBool(Bool=$User.bIsPainted, True Color=$Particles.R1 Color, False Color=<rgba 0 0.25 1.5 1>)
    Index 2 Linear Color:LinearColor=&MakeCustomLinearColorFromBool(Bool=$User.bIsPainted, True Color=$Particles.R1 Color, False Color=<rgba 1.2 0.156 1.113 1>)
    Module.Index0Float0Write=true
    Module.Index0Float1Write=true
    Module.Index0Float2Write=true
    Module.Index0Float3Write=true
    Write Parameter Index 1=::NewEnumerator5
    Write Parameter Index 2=::NewEnumerator5
   )
   /SetVariables(
    Particles.Dynamic_CompleTrail:Vector4f=&MakeVector4(
     X=&SelectFloatFromArray(Direct Array Index=&SpawnGroup(), Float Selection Array=DataInterface() {propertyValues {FloatData [-0.3,0.3]}})
     Y=4
     Z=&MakeCustomFloatFromBool(Bool=$User.bIsPainted, True Float=1, False Float=0)
     W=1.1
    )
    Particles.ScaleWidthComple:NiagaraFloat=&Multiply_Float(
     A=$Particles.WidthComple
     B=&FloatFromCurve(
      FloatCurve=DataInterface() {propertyValues {Curve {keys [
          {interpMode=RCIM_Cubic, tangentMode=RCTM_Break, time=0, value=0.25}
          {arriveTangent=-0.05, interpMode=RCIM_Cubic, leaveTangent=-0.05, time=0.3, value=0.95}
          {interpMode=RCIM_Cubic, tangentMode=RCTM_Break, time=1, value=0.2}
         ]}}}
      Scale Curve=1.2
     )
    )
    Particles.Scale_ColorComple:LinearColor=&VC_DynamicScaleColor(
     Color to Scale:LinearColor=$Particles.Color
     RGB Scale:Vector3f=&VectorFromFloat(Value=&MakeCustomFloatFromBool(
       Bool=$User.bIsPainted
       True Float=&FloatFromCurve(FloatCurve=DataInterface() {propertyValues {Curve {keys [<at 0 0.3>,<at 0.4 1>,<at 0.9 0>]}}}, Scale Curve=2.5)
       False Float=1
      ))
     Alpha Scale:NiagaraFloat=1
    )
   ) {isSetParameters=true}
   &ScaleRibbonWidth(Ribbon Width Scale=&FloatFromCurve(
     FloatCurve=DataInterface() {propertyValues {Curve {keys [
         <at 0 0.9 RCIM_Cubic>
         {interpMode=RCIM_Cubic, tangentMode=RCTM_Break, time=0.35, value=1}
         {arriveTangent=-0.3242, interpMode=RCIM_Cubic, tangentMode=RCTM_Break, time=0.95, value=0.1}
        ]}}}
     Scale Curve=0.95
    ))
   &ScaleColor(
    Scale Alpha=&MakeCustomFloatFromBool(Bool=$User.bIsFrontendPreview, True Float=0.5, False Float=1)
    Scale RGB=&MakeCustomVectorFromBool(Bool=$User.bIsPainted, True Vector=<1 1 1>, False Vector=<2 1.3 1.3>)
    ScaleA=true
    ScaleRGB=true
   )
   /SetVariables(Particles.R1 Color:LinearColor=&VC_DynamicScaleColor(
     Color to Scale:LinearColor=$Particles.R1 Color
     RGB Scale:Vector3f=&VectorFromFloat(Value=&MakeCustomFloatFromBool(
       Bool=$User.bIsPainted
       True Float=1
       False Float=&MakeCustomFloatFromBool(Bool=$User.bIsFrontendPreview, True Float=1, False Float=0.5)
      ))
     Alpha Scale:NiagaraFloat=&FloatFromCurve(FloatCurve=DataInterface() {propertyValues {Curve {keys [
          <at 0 0 RCIM_Cubic>
          {arriveTangent=2.4088, interpMode=RCIM_Cubic, leaveTangent=2.4088, tangentMode=RCTM_User, time=0.1, value=0.25}
          {arriveTangent=0.4231, interpMode=RCIM_Cubic, leaveTangent=0.4231, tangentMode=RCTM_User, time=0.4, value=0.7}
          {arriveTangent=-1.193, interpMode=RCIM_Cubic, leaveTangent=-1.193, tangentMode=RCTM_User, time=0.7, value=0.6}
          {interpMode=RCIM_Cubic, tangentMode=RCTM_User, time=1, value=0}
         ]}}})
    )) {isSetParameters=true}
   &InheritSourceMovement(
    Applied Position Scale=&Multiply_VectorByFloat(Vector=<1 1 1>, Float=1)
    ApplyForce=false
    ApplyPosition=true
    ApplyPositionFalloff=false
   )
   &SolveForcesAndVelocity(Clamp Velocity=false, Limit Acceleration=false)
   &VC_UpdateRibbon_v1(Enable Skeletal Position=true)
  ]
  rendererClasses [&NiagaraRibbonRendererProperties]
  renderers [
   {rendererClass=&NiagaraRibbonRendererProperties}
   {rendererClass=&NiagaraRibbonRendererProperties, rendererIndex=1}
  ]
 }
 NE_VC_WheelWrap {
  emitterUpdateScript [
   &EmitterState(
    Inactive Response=::NewEnumerator1
    Life Cycle Mode=::NewEnumerator1
    Loop Behavior=::NewEnumerator1
    Loop Duration Mode=::NewEnumerator1
    UseLoopDelay=false
   )
   &SpawnBurst_Instantaneous(Spawn Count=2, Use Spawn Probability=false, UseLoopCountLimit=false)
  ]
  particleSpawnScript [
   &VC_PaintedVariantsOverride() {enabled=false}
   &InitializeParticle(
    Mesh Scale Mode=::NewEnumerator4
    Sprite Rotation Mode=::NewEnumerator2
    Sprite Size Mode=::NewEnumerator3
    UsePositionOffset=false
    Write Material Random=true
   )
   &PartitionParticles()
   &SkeletalMeshLocation(
    Bone / Socket Index=&Add_Integer(A=$Particles.PartitionParticles.Partition, B=4)
    Bone / Socket Sampling Mode=::NewEnumerator5
    Interpolate Mesh Sampling=false
    MaskSpawnGroup=false
    Mesh Sampling Type=::NewEnumerator6
    Skeletal Mesh=$System.SkeletalMesh
   )
  ]
  particleUpdateScript [
   &VC_PaintedVariantsOverride(
    BurntSienna=<rgba 0.0642 0.0198 0.0079 1>
    Crimson=<rgba 1 0 0.0271 1>
    R1 Black:LinearColor=<rgba 0.22 0.22 0.22 1>
    R1 BurntSienna:LinearColor=<rgba 0.8 0.1993 0.12 1>
    R1 Cobalt:LinearColor=<rgba 0.125 0.275 1.25 1>
    R1 Crimson:LinearColor=<rgba 1 0.1 0.16 1>
    R1 ForestGreen:LinearColor=<rgba 0.025 0.5 0.0492 1>
    R1 Grey:LinearColor=<rgba 0.4 0.4 0.4 1>
    R1 Lime:LinearColor=<rgba 0.262 0.5 0.025 1>
    R1 Orange:LinearColor=<rgba 1.25 0.1475 0.025 1>
    R1 Pink:LinearColor=<rgba 0.7 0.175 0.525 1>
    R1 Purple:LinearColor=<rgba 0.4259 0.095 1.9 1>
    R1 Saffron:LinearColor=<rgba 0.8 0.62 0.08 1>
    R1 SkyBlue:LinearColor=<rgba 0.1 0.4762 1 1>
    R1 TitaniumWhite:LinearColor=<rgba 0.65 0.65 0.65 1>
    Use R1 Painted Overrides=true
   ) {module=VC_PaintedVariantsOverride002}
   &FN_DisableTireFX(
    Mask Index Attribute:NiagaraInt32=$Particles.PartitionParticles.Partition
    Mask Index FL=2
    Mask Index FR=3
    Mask Index RL=0
    Mask Index RR=1
   )
   &VC_WheelWrapScale(Wheel Index Attribute:NiagaraInt32=&Add_Integer(A=$Particles.PartitionParticles.Partition, B=2))
   &ParticleState(Kill Particles When Lifetime Has Elapsed=false)
   &UpdateMeshOrientation(Rotation Coordinate Space=::NewEnumerator2, Rotation Vector=<0 1 0>) {enabled=false}
   &SkeletalMeshLocation(
    Bone / Socket Index=&Add_Integer(A=$Particles.PartitionParticles.Partition, B=4)
    Bone / Socket Sampling Mode=::NewEnumerator5
    MaskSpawnGroup=false
    Mesh Sampling Type=::NewEnumerator6
    Skeletal Mesh=$System.SkeletalMesh
   ) {module=SkeletalMeshLocation001}
   &DynamicMaterialParameters(
    Index 0 Param 1=&MakeCustomFloatFromBool(Bool=$User.bIsPainted, True Float=1, False Float=0)
    Index 0 Param 2=&MakeCustomFloatFromBool(Bool=$User.bIsPainted, True Float=0, False Float=2.9)
    Index 0 Param 3=&SelectFloatFromArray(Direct Array Index=&SpawnGroup(), Float Selection Array=DataInterface() {propertyValues {FloatData [0.3,0]}})
    Index 1 Linear Color:LinearColor=&MakeCustomLinearColorFromBool(Bool=$User.bIsPainted, True Color=$Particles.R1 Color, False Color=<rgba 0.01 0.34 1 1>)
    Index 2 Linear Color:LinearColor=&MakeCustomLinearColorFromBool(Bool=$User.bIsPainted, True Color=$Particles.R1 Color, False Color=<rgba 1.2 0.18 0.945 1>)
    Module.Index0Float0Write=true
    Module.Index0Float1Write=true
    Module.Index0Float2Write=true
    Module.Index0Float3Write=true
    Write Parameter Index 1=::NewEnumerator5
    Write Parameter Index 2=::NewEnumerator5
   )
   /SetVariables(Particles.R1 ColorScaled:LinearColor=&VC_DynamicScaleColor(
     Color to Scale:LinearColor=$Particles.R1 Color
     RGB Scale:Vector3f=&VectorFromFloat(Value=&MakeCustomFloatFromBool(Bool=$User.bIsPainted, True Float=2, False Float=1.875))
     Alpha Scale:NiagaraFloat=1
    )) {isSetParameters=true}
   &ScaleColor(
    Scale Alpha=&MakeCustomFloatFromBool(Bool=$User.bIsFrontendPreview, True Float=0.5, False Float=1)
    Scale RGB=&MakeCustomVectorFromBool(Bool=$User.bIsPainted, True Vector=<1 1 1>, False Vector=<2 1.3 1.3>)
    ScaleA=true
    ScaleRGB=true
   )
   &SolveForcesAndVelocity(Clamp Velocity=false, Limit Acceleration=false)
  ]
  rendererClasses [&NiagaraMeshRendererProperties]
  renderers [
   {rendererClass=&NiagaraMeshRendererProperties, rendererIndex=0}
   {rendererClass=&NiagaraMeshRendererProperties, rendererIndex=1}
  ]
 }
}
summary {
 systemName=NS_DriftTrail_GenieSmoke_MP
 emitters [
  {emitterName=NE_VC_MainRibbon, rendererClasses [&NiagaraRibbonRendererProperties]}
  {emitterName=NE_VC_WheelWrap, rendererClasses [&NiagaraMeshRendererProperties]}
  {emitterName=NE_VC_DriftTrail_Sprite_Parent, rendererClasses [&NiagaraSpriteRendererProperties]}
  {emitterName=NE_VC_DriftTrail_Sprite_Parent001, rendererClasses [&NiagaraSpriteRendererProperties]}
  {emitterName=NE_VC_DriftTrail_Sprite_Parent002, rendererClasses [&NiagaraSpriteRendererProperties]}
 ]
 userVariables [
  User.bIsFrontendPreview {type=NiagaraBool}
  User.bIsPainted {type=NiagaraBool}
  User.ColorVariantID {type=NiagaraInt32}
  User.offset {type=Vector3f, description="offset from wheels(set through cosmetic template)", defaultValue {struct=Vector3f, value=<0 0 0>}}
  User.PaintedColor {type=LinearColor, description="", defaultValue {struct=LinearColor, value=<rgba 1 1 1 1>}}
  User.PaintedScalar {type=NiagaraFloat, defaultValue=1, description=""}
  User.VehicleFlags {type=NiagaraInt32, defaultValue=806289408}
  User.Wheel_Radius_Array {
   type=NiagaraDataInterfaceArrayFloat
   defaultValue {
    struct=NiagaraExt_VariableValue_DataInterface
    value {
     dataInterface=&self:NiagaraDataInterfaceArrayFloat_1
     dataInterfaceClass=&NiagaraDataInterfaceArrayFloat
    }
   }
  }
  User.Wheel_Width_Array {
   type=NiagaraDataInterfaceArrayFloat
   defaultValue {
    struct=NiagaraExt_VariableValue_DataInterface
    value {
     dataInterface=&self:NiagaraDataInterfaceArrayFloat_0
     dataInterfaceClass=&NiagaraDataInterfaceArrayFloat
    }
   }
  }
  User.WheelTouchingArray {
   type=NiagaraDataInterfaceArrayFloat
   defaultValue {
    struct=NiagaraExt_VariableValue_DataInterface
    value {
     dataInterface=&self:NiagaraDataInterfaceArrayFloat_2
     dataInterfaceClass=&NiagaraDataInterfaceArrayFloat
    }
   }
  }
 ]
}
