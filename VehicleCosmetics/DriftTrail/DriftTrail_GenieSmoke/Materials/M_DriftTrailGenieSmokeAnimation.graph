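# Translucent unlit Niagara ribbon material for the Genie Smoke drift trail vehicle cosmetic, drawing animated, two-tone smoke.
# Chained Ifs on looping time step through foreground and background flipbook frames packed into RGB channels of panned textures.
# Dynamic parameters drive the blue-to-pink gradient, desaturation, offset and flip, while multiplied noises add bright sparkles to emissive.
# Noise-distorted masks and end fades shape opacity, and a sine/cosine WPO group computes a swirling wave.
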Material /VehicleCosmetics/DriftTrail/DriftTrail_GenieSmoke/Materials/M_DriftTrailGenieSmokeAnimation
graph {
 properties {
  bAutomaticallySetUsageInEditor=false
  bUsedWithNiagaraRibbons=true
  blendMode=BLEND_Translucent
  shadingModel=MSM_Unlit
  usedShadingModels=Unlit
 }
 nodes [
  n1 = DynamicParameter() {parameterIndex=1, defaultValue=<rgba 0.01 0.2 1 1>, paramNames [R_BlueColor,G_BlueColor,B_BlueColor,Param4]}
  n2 = DynamicParameter() {parameterIndex=2, defaultValue=<rgba 1 0.15 1 1>, paramNames [R_PinkColor,G_PinkColor,B_PinkColor,Param4]}
  n3 = OneMinus(None=ComponentMask(None=TextureCoordinate()) {r=true})
  n4 = DynamicParameter() {parameterIndex=3, paramNames [Offset, PowerBlueColor, Desaturation, FlipT]}
  n5 = Desaturation(None=LinearInterpolate(A=@n1.RGB, B=@n2.RGB, Alpha=Power(Base=@n3, Exp=@n4.PowerBlueColor) {constExponent=0.3}), Fraction=@n4.Desaturation)
  n6 = DynamicParameter() {parameterIndex=3, paramNames [Offset, PowerBlueColor, Desaturation, FlipT]}
  n7 = TextureCoordinate()
  n8 = Add(
   A=AppendVector(A=@n6.Offset, B=Constant())
   B=If(A=@n6.FlipT, B=Constant() {r=1}, "A > B"=@n7, "A == B"=@n7, "A < B"=TextureCoordinate() {vTiling=-1})
  )
  n9 = ScalarParameter() {defaultValue=2, parameterName=TileUV}
  n10 = Constant() {r=1}
  n11 = Panner(Coordinate=Multiply(A=@n8, B=AppendVector(A=@n9, B=@n10))) {speedX=-0.75}
  n12 = TextureSample(UVs=@n11) {texture=/VehicleCosmetics/Boosts/Boost_GenieSmoke/Texture/T_VCVFX_Boost_GenieSmoke_BGAnimation_06_Blended_RGB}
  n13 = Constant() {r=5.3}
  n14 = Multiply(A=Frac(None=Time()), B=Constant() {r=16})
  n15 = Constant() {r=15}
  n16 = Constant() {r=14}
  n17 = TextureSample(UVs=@n11) {texture=/VehicleCosmetics/Boosts/Boost_GenieSmoke/Texture/T_VCVFX_Boost_GenieSmoke_Animation_05_RGB}
  n18 = Constant() {r=13}
  n19 = Constant() {r=12}
  n20 = Constant() {r=11}
  n21 = TextureSample(UVs=@n11) {texture=/VehicleCosmetics/Boosts/Boost_GenieSmoke/Texture/T_VCVFX_Boost_GenieSmoke_Animation_04_RGB}
  n22 = Constant() {r=10}
  n23 = Constant() {r=9}
  n24 = Constant() {r=8}
  n25 = TextureSample(UVs=@n11) {texture=/VehicleCosmetics/Boosts/Boost_GenieSmoke/Texture/T_VCVFX_Boost_GenieSmoke_Animation_03_RGB}
  n26 = Constant() {r=7}
  n27 = Constant() {r=6}
  n28 = Constant() {r=5}
  n29 = TextureSample(UVs=@n11) {texture=/VehicleCosmetics/Boosts/Boost_GenieSmoke/Texture/T_VCVFX_Boost_GenieSmoke_Animation_02_RGB}
  n30 = Constant() {r=4}
  n31 = Constant() {r=3}
  n32 = Constant() {r=2}
  n33 = TextureSample(UVs=@n11) {texture=/VehicleCosmetics/Boosts/Boost_GenieSmoke/Texture/T_VCVFX_Boost_GenieSmoke_Animation_01_RGB}
  n34 = Constant() {r=1}
  n35 = If(A=@n14, B=@n34, "A > B"=@n33.G, "A == B"=@n33.G, "A < B"=@n33.G)
  n36 = If(A=@n14, B=@n32, "A > B"=@n33.B, "A == B"=@n35, "A < B"=@n35)
  n37 = If(A=@n14, B=@n31, "A > B"=@n29.R, "A == B"=@n36, "A < B"=@n36)
  n38 = If(A=@n14, B=@n30, "A > B"=@n29.G, "A == B"=@n37, "A < B"=@n37)
  n39 = If(A=@n14, B=@n28, "A > B"=@n29.B, "A == B"=@n38, "A < B"=@n38)
  n40 = If(A=@n14, B=@n27, "A > B"=@n25.R, "A == B"=@n39, "A < B"=@n39)
  n41 = If(A=@n14, B=@n26, "A > B"=@n25.G, "A == B"=@n40, "A < B"=@n40)
  n42 = If(A=@n14, B=@n24, "A > B"=@n25.B, "A == B"=@n41, "A < B"=@n41)
  n43 = If(A=@n14, B=@n23, "A > B"=@n21.R, "A == B"=@n42, "A < B"=@n42)
  n44 = If(A=@n14, B=@n22, "A > B"=@n21.G, "A == B"=@n43, "A < B"=@n43)
  n45 = If(A=@n14, B=@n20, "A > B"=@n21.B, "A == B"=@n44, "A < B"=@n44)
  n46 = If(A=@n14, B=@n19, "A > B"=@n17.R, "A == B"=@n45, "A < B"=@n45)
  n47 = If(A=@n14, B=@n18, "A > B"=@n17.G, "A == B"=@n46, "A < B"=@n46)
  n48 = If(A=@n14, B=@n16, "A > B"=@n17.B, "A == B"=@n47, "A < B"=@n47)
  n49 = If(A=@n14, B=@n15, "A > B"=@n12.R, "A == B"=@n48, "A < B"=@n48)
  n50 = TextureSample(UVs=@n11) {texture=/VehicleCosmetics/Boosts/Boost_GenieSmoke/Texture/T_VCVFX_Boost_GenieSmoke_BGAnimation_05_RGB}
  n51 = TextureSample(UVs=@n11) {texture=/VehicleCosmetics/Boosts/Boost_GenieSmoke/Texture/T_VCVFX_Boost_GenieSmoke_BGAnimation_04_RGB}
  n52 = TextureSample(UVs=@n11) {texture=/VehicleCosmetics/Boosts/Boost_GenieSmoke/Texture/T_VCVFX_Boost_GenieSmoke_BGAnimation_03_RGB}
  n53 = TextureSample(UVs=@n11) {texture=/VehicleCosmetics/Boosts/Boost_GenieSmoke/Texture/T_VCVFX_Boost_GenieSmoke_BGAnimation_02_RGB}
  n54 = TextureSample(UVs=@n11) {texture=/VehicleCosmetics/Boosts/Boost_GenieSmoke/Texture/T_VCVFX_Boost_GenieSmoke_BGAnimation_01_RGB}
  n55 = If(A=@n14, B=@n34, "A > B"=@n54.G, "A == B"=@n54.G, "A < B"=@n54.G)
  n56 = If(A=@n14, B=@n32, "A > B"=@n54.B, "A == B"=@n55, "A < B"=@n55)
  n57 = If(A=@n14, B=@n31, "A > B"=@n53.R, "A == B"=@n56, "A < B"=@n56)
  n58 = If(A=@n14, B=@n30, "A > B"=@n53.G, "A == B"=@n57, "A < B"=@n57)
  n59 = If(A=@n14, B=@n28, "A > B"=@n53.B, "A == B"=@n58, "A < B"=@n58)
  n60 = If(A=@n14, B=@n27, "A > B"=@n52.R, "A == B"=@n59, "A < B"=@n59)
  n61 = If(A=@n14, B=@n26, "A > B"=@n52.G, "A == B"=@n60, "A < B"=@n60)
  n62 = If(A=@n14, B=@n24, "A > B"=@n52.B, "A == B"=@n61, "A < B"=@n61)
  n63 = If(A=@n14, B=@n23, "A > B"=@n51.R, "A == B"=@n62, "A < B"=@n62)
  n64 = If(A=@n14, B=@n22, "A > B"=@n51.G, "A == B"=@n63, "A < B"=@n63)
  n65 = If(A=@n14, B=@n20, "A > B"=@n51.B, "A == B"=@n64, "A < B"=@n64)
  n66 = If(A=@n14, B=@n19, "A > B"=@n50.R, "A == B"=@n65, "A < B"=@n65)
  n67 = If(A=@n14, B=@n18, "A > B"=@n50.G, "A == B"=@n66, "A < B"=@n66)
  n68 = If(A=@n14, B=@n16, "A > B"=@n50.B, "A == B"=@n67, "A < B"=@n67)
  n69 = If(A=@n14, B=@n15, "A > B"=@n12.B, "A == B"=@n68, "A < B"=@n68)
  n70 = Panner(Coordinate=TextureCoordinate() {uTiling=10, vTiling=10}) {speedX=-0.1}
  n71 = TextureCoordinate() {uTiling=10, vTiling=10}
  n72 = AppendVector(A=@n4.Offset, B=Constant())
  n73 = Constant() {r=1}
  n74 = Constant() {r=0.1}
  n75 = Panner(Coordinate=Multiply(A=Add(A=@n71, B=@n72), B=AppendVector(A=@n73, B=@n74))) {speedX=-1.2}
  MaterialExpressionConstantClamp_1 = Saturate(None=Multiply(
    A=TextureSample(UVs=@n70) {texture=/DealershipSharedLibrary/FX/Textures/Noise/T_VCVFX_NoiseGeneric01Pack_RGB}.B
    B=TextureSample(UVs=@n75) {texture=&T_VCVFX_NoiseGeneric01Pack_RGB}.B
   ))
  n76 = Multiply(A=@MaterialExpressionConstantClamp_1, B=Constant() {r=10024})
  n77 = Add(A=Multiply(A=@n12.G, B=Constant()), B=Add(A=Multiply(A=@n13, B=@n49), B=Add(A=@n69, B=@n76)))
  n78 = ParticleColor()
  MaterialFunctionCall(Emissive=Multiply(A=Multiply(A=@n5, B=@n77), B=@n78.RGB)) {materialFunction=&MF_FX_EmissiveAdjustments}
  n79 = TextureCoordinate()
  n80 = Multiply(A=OneMinus(None=Time()), B=ScalarParameter() {group=WPO, parameterName=Speed})
  n81 = Divide(A=ComponentMask(None=TextureCoordinate()) {r=true}, B=ScalarParameter() {defaultValue=5, group=WPO, parameterName=tile})
  n82 = AppendVector(A=Cosine(None=Add(A=@n80, B=@n81)), B=Sine(None=Add(A=@n81, B=@n80)))
  n83 = Constant()
  n84 = Multiply(
   A=Multiply(A=AppendVector(A=@n82, B=@n83), B=ScalarParameter() {group=WPO, parameterName=Intensity})
   B=ScalarParameter() {group=WPO, parameterName=Amount}
  )
  Multiply(A=Power(Base=ComponentMask(None=@n79) {r=true}, Exp=Constant() {r=0.7}), B=ConstantBiasScale(None=Transform(None=@n84)))
  n85 = AppendVector(A=Constant() {r=0.6}, B=Constant() {r=0.8})
  n86 = Multiply(A=TextureCoordinate(), B=AppendVector(A=Constant() {r=5}, B=Constant() {r=3}))
  n87 = ComponentMask(None=TextureSample(UVs=Panner(Coordinate=@n86) {speedX=-0.8, speedY=-0.8}) {texture=/DealershipSharedLibrary/FX/Textures/Noise/T_VCVFX_Boost_alpine_infinite_Noise_Heightmap_Alpha}.RGB) {g=true, r=true}
  n88 = Constant() {r=0.3}
  n89 = Panner(Coordinate=Add(A=Multiply(A=@n8, B=@n85), B=Multiply(A=@n87, B=@n88))) {speedX=-0.5}
  MaterialExpressionConstantClamp_3 = Saturate(None=Multiply(
    A=Constant() {r=0.5}
    B=TextureSample(UVs=@n89) {texture=/DealershipSharedLibrary/FX/Textures/Gradient_Masks/T_VCVFX_boost_Space_Tear_A_T}.RGB
   ))
  n90 = Multiply(A=@n8, B=AppendVector(A=Constant() {r=4}, B=Constant() {r=1}))
  n91 = Power(
   Base=TextureSample(UVs=Panner(Coordinate=@n90) {speedX=-1.5}) {texture=/VehicleCosmetics/Boosts/Boost_GenieSmoke/Texture/T_VCVFX_Boost_GenieSmoke_Trail2_RGB}.RGB
   Exp=Constant() {r=2}
  )
  n92 = Multiply(A=Add(A=@n49, B=@n69), B=Multiply(A=OneMinus(None=@MaterialExpressionConstantClamp_3), B=OneMinus(None=@n91)))
  n93 = ComponentMask(None=TextureCoordinate()) {r=true}
  n94 = Power(Base=OneMinus(None=@n93), Exp=ScalarParameter() {defaultValue=2.2, parameterName=power fade one minus})
  n95 = Multiply(A=Power(Base=@n93, Exp=Constant() {r=2}), B=Constant() {r=50})
  Saturate(None=Multiply(A=Multiply(A=@n78.A, B=@n92), B=Multiply(A=@n94, B=@n95)))
 ]
}
