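# Translucent unlit Niagara sprite material for droplets in the Genie Smoke drift trail cosmetic, tinted by particle color through an emissive adjustment function.
# Two panning cell-noise samples in radial coordinates combine into a small distortion that wobbles the procedural shape.
# Sine, dot product and ceil math cut two hard-edged perpendicular shapes, each limited by a one-axis sphere mask.
# These add to a small exponential radial gradient, and particle alpha scales the resulting mask.
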Material /VehicleCosmetics/DriftTrail/DriftTrail_GenieSmoke/Materials/M_DriftTrailGenieSmoke_Droplet
graph {
 properties {
  bAutomaticallySetUsageInEditor=false
  bUsedWithNiagaraSprites=true
  blendMode=BLEND_Translucent
  shadingModel=MSM_Unlit
  usedShadingModels=Unlit
 }
 nodes [
  n1 = ParticleColor()
  MaterialFunctionCall(Emissive=@n1.RGB) {materialFunction=&MF_FX_EmissiveAdjustments}
  n2 = MaterialFunctionCall() {materialFunction=&VectorToRadialValue}
  n3 = Panner(Coordinate=Multiply(A=@n2.Radial Coordinates) {constB=0.1}) {speedX=-0.075, speedY=-0.075}
  n4 = Power(Base=ConstantBiasScale(None=TextureSample(UVs=@n3) {texture=/DealershipSharedLibrary/FX/Textures/Noise/T_VCVFX_CellNoiseSkins_A}.R))
  n5 = Panner(Coordinate=Multiply(A=@n2.Radial Coordinates) {constB=0.25}) {speedX=0.125, speedY=-0.065}
  n6 = Power(Base=ConstantBiasScale(None=TextureSample(UVs=@n5) {texture=&T_VCVFX_CellNoiseSkins_A}.R))
  n7 = Multiply(A=Saturate(None=Add(A=@n4, B=@n6))) {constB=0.025}
  n8 = TextureCoordinate()
  n9 = ScalarParameter() {defaultValue=1, parameterName=Scale}
  n10 = Multiply(A=@n9) {constB=0.5}
  n11 = Abs(None=Subtract(A=Multiply(A=@n8, B=@n9), B=AppendVector(A=@n10, B=@n10)))
  n12 = Constant2Vector() {g=0.09}
  n13 = Power(Base=Add(A=@n7, B=Subtract(A=@n11, B=@n12))) {constExponent=2.2}
  n14 = DotProduct(A=Sine(None=Multiply(A=@n13) {constB=8}) {period=2.5}, B=Constant2Vector() {g=0.1, r=-2})
  n15 = Multiply(A=Saturate(None=Ceil(None=@n14)), B=SphereMask(A=TextureCoordinate() {vTiling=0}, B=Constant2Vector() {r=0.5}) {attenuationRadius=0.3})
  n16 = Constant2Vector() {r=0.09}
  n17 = Power(Base=Add(A=@n7, B=Subtract(A=@n11, B=@n16) {constB=0.1})) {constExponent=2.2}
  n18 = DotProduct(A=Sine(None=Multiply(A=@n17) {constB=8}) {period=2.5}, B=Constant2Vector() {g=-2, r=0.1})
  n19 = Multiply(A=Saturate(None=Ceil(None=@n18)), B=SphereMask(A=TextureCoordinate() {uTiling=0}, B=Constant2Vector() {g=0.5}) {attenuationRadius=0.3})
  n20 = Add(A=@n7, B=TextureCoordinate() {uTiling=0.95, vTiling=0.95})
  n21 = Constant() {r=0.05}
  n22 = Constant() {r=100}
  Multiply(
   A=@n1.A
   B=Add(
    A=Add(A=@n15, B=@n19)
    B=MaterialFunctionCall(UVs=@n20, Radius=@n21, Density=@n22) {materialFunction=&RadialGradientExponential}.RadialGradientExponential
   )
  )
 ]
}
