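# Translucent unlit Niagara sprite material for the swirl spirals in the Genie Smoke drift trail.
# Two SubUV spritesheets, swirl and inside swirl, are weighted, summed and split into channels.
# The fraction of dynamic parameter R selects blue above 0.66, otherwise green; the 0.33 test returns green on every branch, leaving red unused.
# That mask, scaled by particle color and alpha, feeds emissive-adjustment and depth-fade opacity functions.
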Material /VehicleCosmetics/DriftTrail/DriftTrail_GenieSmoke/Materials/M_DriftTrailGenieSmoke_Spirals
graph {
 properties {
  bAutomaticallySetUsageInEditor=false
  bUsedWithNiagaraSprites=true
  blendMode=BLEND_Translucent
  shadingModel=MSM_Unlit
  usedShadingModels=Unlit
 }
 nodes [
  n1 = ParticleColor()
  n2 = Frac(None=DynamicParameter() {paramNames [R,G,B,A]}.R)
  n3 = ParticleSubUV!(UVs, Tex, Apply View MipBias) {texture=/VehicleCosmetics/DriftTrail/DriftTrail_GenieSmoke/Textures/T_SwirlSpritesheet}
  n4 = ParticleSubUV() {texture=/VehicleCosmetics/DriftTrail/DriftTrail_GenieSmoke/Textures/T_SwirlSpritesheet_Inside}
  n5 = MaterialFunctionCall(Float3=Add(A=Multiply(A=@n3.RGB) {constB=2.5}, B=Multiply(A=@n4.RGB) {constB=0.15})) {materialFunction=&BreakOutFloat3Components}
  n6 = If(A=@n2, "A > B"=@n5.G, "A == B"=@n5.G, "A < B"=@n5.G) {constB=0.33}
  n7 = Saturate(None=If(A=@n2, "A > B"=@n5.B, "A == B"=@n6, "A < B"=@n6) {constB=0.66})
  MaterialFunctionCall(Emissive=Multiply(A=@n1.RGB, B=@n7)) {materialFunction=&MF_FX_EmissiveAdjustments}
  n8 = Saturate(None=Multiply(A=@n7) {constB=3})
  MaterialFunctionCall(Opacity In=Saturate(None=Multiply(A=@n1.A, B=@n8)), DepthFade Distance=Constant() {r=100}) {materialFunction=&MF_VCVFX_DepthFade}
 ]
}
